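; Gameplay ability for a Season 28 Battle Pass quest about hiding in Creepin' Cardboard, signaling that action through a custom verb message.
; On activation from a gameplay event it calls the parent, then broadcasts a message built from the owning player pawn and the VerbName variable.
; It ends immediately afterward and holds no hiding detection or progress logic in this graph.
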(event Ability|EventActivateAbilityFromEvent
  (params (EventData :type Gameplay%20Event%20Data%20Structure%20%28by%20ref%29))
  (|CallParentFunction EventData)
  (CustomVerb|BroadcastCustomVerbMessage (CustomVerb|MakeCustomVerbMessage (Variables|GAVars|Data|GetOwningPlayerPawn) (Variables|Default|GetVerb%20Name)))
  (Ability|EndAbility))
